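; Gameplay ability for a challenge quest augment that grants a pickup damage effect, apparently when a quest event triggers it.
; On event activation it runs the parent logic, then applies the GE_PickupDMG effect at level 0 to the owning player pawn's ability system component.
; It stores the returned effect handle in GE Handle and ends immediately; removal of the effect is not handled in this graph.
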(event Ability|EventActivateAbilityFromEvent
  (params (EventData :type Gameplay%20Event%20Data%20Structure%20%28by%20ref%29))
  (|CallParentFunction EventData)
  (bind _returnvalue (GameplayEffects|ApplyGameplayEffectToSelf :self (Ability|GetAbilitySystemComponent (Variables|GAVars|Data|GetOwningPlayerPawn)) :GameplayEffectClass (Variables|Default|GetGE_PickupDMG) :Level 0.0))
  (Variables|Default|SetGE%20Handle _returnvalue)
  (Ability|EndAbility))
